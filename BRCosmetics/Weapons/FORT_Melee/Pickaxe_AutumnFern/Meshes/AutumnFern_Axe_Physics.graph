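# Physics asset for the AutumnFern two-handed pickaxe mesh, simulating a bow with two dangling laces on the weapon.
# Capsule bodies cover the handle, the bow base and a two-segment lace chain on each side.
# The bow base is constrained to the handle, and each lace chain hangs from it through swing-limited joints with zero twist limits.
# Left and right laces use slightly different capsule sizes and swing limits.
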PhysicsAsset /BRCosmetics/Weapons/FORT_Melee/Pickaxe_AutumnFern/Meshes/AutumnFern_Axe_Physics
bodies {
 Dyn_bowLace_L_01 [{shapeType=Capsule, length=0.99, radius=0.5719, center=<0.3 0 0>, rotation=<rot 90 0 0>}]
 Dyn_bowLace_L_02 [{shapeType=Capsule, length=0.3236, radius=1.3654, center=<1 0 0>, rotation=<rot 90 0 0>}]
 Dyn_bowLace_R_01 [{shapeType=Capsule, length=0.973, radius=0.8242, center=<0.3 0 0>, rotation=<rot 90 0 0>}]
 Dyn_bowLace_R_02 [{shapeType=Capsule, length=0.7903, radius=1.2002, center=<1 0 0>, rotation=<rot 90 0 0>}]
 Melee_TwoHanded_Handle [{shapeType=Capsule, length=1, radius=1, center=<1 0 0>, rotation=<rot 90 0 0>}]
 base_bow [{shapeType=Capsule, length=0, radius=1.3699, center=<-0.0366 -0.4264 0.3944>, rotation=<rot -0.0003 -88.4151 177.7137>}]
}
constraints [
 {bone1Name=base_bow, bone2Name=Melee_TwoHanded_Handle, swing1LimitDegrees=45, swing2LimitDegrees=45, twistLimitDegrees=0}
 {bone1Name=Dyn_bowLace_R_01, bone2Name=base_bow, swing1LimitDegrees=15, swing2LimitDegrees=25, twistLimitDegrees=0}
 {bone1Name=Dyn_bowLace_R_02, bone2Name=Dyn_bowLace_R_01, swing1LimitDegrees=25, swing2LimitDegrees=25, twistLimitDegrees=0}
 {bone1Name=Dyn_bowLace_L_01, bone2Name=base_bow, swing1LimitDegrees=15, swing2LimitDegrees=15, twistLimitDegrees=0}
 {bone1Name=Dyn_bowLace_L_02, bone2Name=Dyn_bowLace_L_01, swing1LimitDegrees=20, swing2LimitDegrees=20, twistLimitDegrees=0}
]
