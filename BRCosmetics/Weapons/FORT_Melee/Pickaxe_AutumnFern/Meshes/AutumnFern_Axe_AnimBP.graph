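# Animation Blueprint for the AutumnFern pickaxe that adds secondary physics motion to the cosmetic weapon's skeleton.
# The input pose runs through four chained AnimDynamics nodes for the left and right bow tie bones, with angular springs and limits.
# A RigidBody node then simulates the axe physics asset relative to b_AR_Root.
# Each update copies front-end sim-space settings into that node's variable when bIsInFrontEnd is true; otherwise it reassigns the current value.
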AnimBlueprint /BRCosmetics/Weapons/FORT_Melee/Pickaxe_AutumnFern/Meshes/AutumnFern_Axe_AnimBP
target_skeleton=&/AutumnFern_Axe_Skeleton
graphs [
 AnimGraph {nodes [
   AnimGraphNode_SubInput_0 = Input Pose()
   n1 = Anim Dynamics - Bone: Dyn_bowTie_L_01(ComponentPose=Local To Component(LocalPose=@AnimGraphNode_SubInput_0), Alpha=1, ExternalForce="0,0,0") {
    angularDampingOverride=1
    angularSpringConstant=10000
    bAngularSpring=true
    bOverrideAngularDamping=true
    bUseGravityOverride=true
    gravityScale=0.35
    simulationSpace=RootRelative
    alphaBoolBlend {blendOption=Linear}
    boundBone {boneName=Dyn_bowTie_L_01}
    componentAppliedLinearAccClamp=<3000 3000 3000>
    componentLinearAccScale=<0.4 0.4 0.4>
    componentLinearVelScale=<0.4 0.4 0.4>
    gravityOverride {z=-980}
    physicsBodyDefinitions [{
      boundBone {boneName=Dyn_bowTie_L_01}
      boxExtents=<20 20 20>
      constraintSetup {angularLimitsMax=<2 5 0>, angularLimitsMin=<-2 -5 -30>, angularTarget=<1 0 0>}
      localJointOffset=<4 0 0>
     }]
    simSpaceSettings {maxAngularAcceleration=50, maxAngularVelocity=10, simSpaceAngularAlpha=0.5}
   }
   n2 = Anim Dynamics - Bone: Dyn_bowTie_L_02(ComponentPose=@n1, Alpha=1, ExternalForce="0,0,0") {
    angularDampingOverride=1
    angularSpringConstant=10000
    bAngularSpring=true
    bOverrideAngularDamping=true
    bUseGravityOverride=true
    gravityScale=0.35
    simulationSpace=RootRelative
    alphaBoolBlend {blendOption=Linear}
    boundBone {boneName=Dyn_bowTie_L_02}
    componentAppliedLinearAccClamp=<3000 3000 3000>
    componentLinearAccScale=<0.4 0.4 0.4>
    componentLinearVelScale=<0.4 0.4 0.4>
    gravityOverride {z=-980}
    physicsBodyDefinitions [{
      boundBone {boneName=Dyn_bowTie_L_02}
      boxExtents=<20 20 20>
      constraintSetup {angularLimitsMax=<4 7 0>, angularLimitsMin=<-4 -7 -30>, angularTarget=<1 0 0>}
      localJointOffset=<4 0 0>
     }]
    simSpaceSettings {maxAngularAcceleration=50, maxAngularVelocity=10, simSpaceAngularAlpha=0.5}
   }
   n3 = Anim Dynamics - Bone: Dyn_bowTie_R_01(ComponentPose=@n2, Alpha=1, ExternalForce="0,0,0") {
    angularDampingOverride=1
    angularSpringConstant=10000
    bAngularSpring=true
    bOverrideAngularDamping=true
    bUseGravityOverride=true
    gravityScale=0.35
    simulationSpace=RootRelative
    alphaBoolBlend {blendOption=Linear}
    boundBone {boneName=Dyn_bowTie_R_01}
    componentAppliedLinearAccClamp=<3000 3000 3000>
    componentLinearAccScale=<0.4 0.4 0.4>
    componentLinearVelScale=<0.4 0.4 0.4>
    gravityOverride {z=-980}
    physicsBodyDefinitions [{
      boundBone {boneName=Dyn_bowTie_R_01}
      boxExtents=<20 20 20>
      constraintSetup {angularLimitsMax=<2 5 0>, angularLimitsMin=<-2 -5 -30>, angularTarget=<1 0 0>}
      localJointOffset=<4 0 0>
     }]
    simSpaceSettings {maxAngularAcceleration=50, maxAngularVelocity=10, simSpaceAngularAlpha=0.5}
   }
   n4 = Anim Dynamics - Bone: Dyn_bowTie_R_02(ComponentPose=@n3, Alpha=1, ExternalForce="0,0,0") {
    angularDampingOverride=1
    angularSpringConstant=10000
    bAngularSpring=true
    bOverrideAngularDamping=true
    bUseGravityOverride=true
    gravityScale=0.35
    simulationSpace=RootRelative
    alphaBoolBlend {blendOption=Linear}
    boundBone {boneName=Dyn_bowTie_R_02}
    componentAppliedLinearAccClamp=<3000 3000 3000>
    componentLinearAccScale=<0.4 0.4 0.4>
    componentLinearVelScale=<0.4 0.4 0.4>
    gravityOverride {z=-980}
    physicsBodyDefinitions [{
      boundBone {boneName=Dyn_bowTie_R_02}
      boxExtents=<20 20 20>
      constraintSetup {angularLimitsMax=<4 7 0>, angularLimitsMin=<-4 -7 -30>, angularTarget=<1 0 0>}
      localJointOffset=<4 0 0>
     }]
    simSpaceSettings {maxAngularAcceleration=50, maxAngularVelocity=10, simSpaceAngularAlpha=0.5}
   }
   n5 = RigidBody!(
    ComponentPose=@n4
    Alpha
    ExternalForce
    SimSpaceSettings:Sim Space Settings Structure=Get Sim Space Settings() -> (Sim Space Settings:Sim Space Settings Structure)
   ) -> (Pose) {
    bOverrideWorldGravity=true
    overridePhysicsAsset=&/AutumnFern_Axe_Physics
    simulationSpace=BaseBoneSpace
    alphaBoolBlend {blendOption=Linear}
    baseBoneRef {boneName=b_AR_Root}
    overrideWorldGravity {z=-980}
    simSpaceSettings {
     dampingAlpha=0
     maxAngularAcceleration=30
     maxAngularVelocity=6
     maxLinearAcceleration=1000
     maxLinearVelocity=200
     worldAlpha=1
     externalLinearDragV=<0.2 0.2 0.2>
    }
   }
   Output Pose(Result=Component To Local(ComponentPose=@n5)) {layerGroup=None}
  ]}
 EventGraph {nodes [
   Event BlueprintUpdateAnimation()
   > n1 = Branch(Condition=Get bIsInFrontEnd())
   > Set Sim Space Settings(Sim Space Settings:Sim Space Settings Structure=Get Sim Space Settings_FE() -> (Sim Space Settings_FE:Sim Space Settings Structure)) -> (Output_Get:Sim Space Settings Structure)
   Set Sim Space Settings(execute=@n1.else, Sim Space Settings=Get Sim Space Settings())
   TryGetPawnOwner()
  ]}
]
variables [
 Sim Space Settings {
  type=SimSpaceSettings
  default {
   dampingAlpha=0
   maxAngularAcceleration=40
   maxAngularVelocity=9
   maxLinearAcceleration=1000
   maxLinearVelocity=200
   velocityScaleZ=1
   worldAlpha=1
   externalLinearDragV=<0.2 0.2 0.2>
  }
 }
 Sim Space Settings_FE {
  type=SimSpaceSettings
  default {
   dampingAlpha=0
   maxAngularAcceleration=700
   maxAngularVelocity=200
   maxLinearAcceleration=500
   maxLinearVelocity=400
   velocityScaleZ=1
   worldAlpha=0.8
   externalLinearDragV=<0.1 0.1 0.1>
  }
 }
]
